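# Physics asset for the Sunrise Palace backpack cosmetic, defining capsule bodies and joints for the pack and its dangling bag, handle, knot and cloth pieces.
# Short chains hang from the pack body: two handle bones, two bag bones, two bottom cloth bones, and a knot that carries two top cloth bones.
# Angular limits keep twist and one swing axis tight, allowing wider sway on the second swing axis, except at the nearly rigid knot.
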PhysicsAsset /BRCosmetics/Accessories/FORT_Backpacks/Backpack_M_MED_Sunrise_Palace/Meshes/M_MED_Sunrise_Palace_Pack_Physics
bodies {
 dyn_bag_01 [{
   shapeType=Capsule
   length=13.633
   radius=5.1471
   shapeName=dyn_bag_01_capsule
   center=<4.9738 0.3115 -0.0463>
   rotation=<rot 0 -0.0002 -89.9984>
  }]
 dyn_bag_02 [{
   shapeType=Capsule
   length=9.0547
   radius=4.2991
   shapeName=dyn_bag_02_capsule
   center=<3.815 0.2113 -0.2714>
   rotation=<rot 0 -179.9916 -89.9999>
  }]
 dyn_cloth_bottom_01 [{
   shapeType=Capsule
   length=1.9284
   radius=0.9182
   shapeName=dyn_cloth_bottom_01_capsule
   center=<0.9828 0.1018 -0.4516>
   rotation=<rot 0 176.8816 -89.418>
  }]
 dyn_cloth_bottom_02 [{
   shapeType=Capsule
   length=1.7233
   radius=1.0953
   shapeName=dyn_cloth_bottom_02_capsule
   center=<2.6764 -0.2298 -0.2968>
   rotation=<rot 0 -0.077 -86.1276>
  }]
 dyn_cloth_top_01 [{
   shapeType=Capsule
   length=2.2302
   radius=0.4927
   shapeName=dyn_cloth_top_01_capsule
   center=<0.532 0.0921 -0.1715>
   rotation=<rot 0 175.4888 -89.8719>
  }]
 dyn_cloth_top_02 [{
   shapeType=Capsule
   length=2.0954
   radius=0.8434
   shapeName=dyn_cloth_top_02_capsule
   center=<0.6117 0.4664 -0.6347>
   rotation=<rot 0 -176.0955 -84.9635>
  }]
 dyn_handle_01 [{
   shapeType=Capsule
   length=5.4292
   radius=0.9661
   shapeName=dyn_handle_01_capsule
   center=<0.8532 0.0034 -0.0483>
   rotation=<rot 0 -2.0554 -88.1883>
  }]
 dyn_handle_02 [{
   shapeType=Capsule
   length=4.221
   radius=1.1659
   shapeName=dyn_handle_02_capsule
   center=<1.0183 0.1648 0.0014>
   rotation=<rot 0 5.8274 -88.7962>
  }]
 dyn_knot [{
   shapeType=Capsule
   length=1.9146
   radius=3.9945
   shapeName=dyn_knot_capsule
   center=<3.0702 -0.3177 -0.2011>
   rotation=<rot 0 -0.0128 -84.2659>
  }]
 pack [{
   shapeType=Capsule
   length=8.2894
   radius=8.3322
   shapeName=pack_capsule
   center=<-2.7467 -4.7502 1.3985>
   rotation=<rot -3.5006 -90.5325 3.0507>
  }]
}
constraints [
 {bone1Name=dyn_handle_02, bone2Name=dyn_handle_01, swing1LimitDegrees=5, swing2LimitDegrees=20, twistLimitDegrees=1}
 {bone1Name=dyn_cloth_top_01, bone2Name=dyn_knot, swing1LimitDegrees=5, swing2LimitDegrees=25, twistLimitDegrees=1}
 {bone1Name=dyn_cloth_top_02, bone2Name=dyn_cloth_top_01, swing1LimitDegrees=5, swing2LimitDegrees=40, twistLimitDegrees=1}
 {bone1Name=dyn_bag_02, bone2Name=dyn_bag_01, swing1LimitDegrees=1, swing2LimitDegrees=15, twistLimitDegrees=1}
 {bone1Name=dyn_cloth_bottom_02, bone2Name=dyn_cloth_bottom_01, swing1LimitDegrees=5, swing2LimitDegrees=30, twistLimitDegrees=1}
 {bone1Name=dyn_handle_01, bone2Name=pack, swing1LimitDegrees=5, swing2LimitDegrees=20, twistLimitDegrees=1}
 {bone1Name=dyn_knot, bone2Name=pack, swing1LimitDegrees=1, swing2LimitDegrees=2, twistLimitDegrees=1}
 {bone1Name=dyn_bag_01, bone2Name=pack, swing1LimitDegrees=1, swing2LimitDegrees=15, twistLimitDegrees=1}
 {bone1Name=dyn_cloth_bottom_01, bone2Name=pack, swing1LimitDegrees=5, swing2LimitDegrees=30, twistLimitDegrees=1}
]
