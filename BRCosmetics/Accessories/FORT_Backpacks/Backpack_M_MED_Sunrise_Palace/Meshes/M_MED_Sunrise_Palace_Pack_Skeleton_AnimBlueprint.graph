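# Animation Blueprint for the Sunrise Palace backpack cosmetic, adding secondary physics motion and an offset that follows the wearer's spine.
# Two RigidBody simulations share a cached input pose and blend on bIsSkydiveFloating.
# The floating branch simulates relative to the pack bone and rolls dyn_bag_02; the other simulates relative to root.
# Each update maps the owner's spine_01-to-spine_05 yaw delta from ref pose into additive root roll and Y translation.
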AnimBlueprint /BRCosmetics/Accessories/FORT_Backpacks/Backpack_M_MED_Sunrise_Palace/Meshes/M_MED_Sunrise_Palace_Pack_Skeleton_AnimBlueprint
target_skeleton=&/M_MED_Sunrise_Palace_Pack_Skeleton
graphs [
 AnimGraph {nodes [
   n1 = RigidBody(ComponentPose=Local To Component(LocalPose=Use cached pose 'SavedPose21755'())) {
    bOverrideWorldGravity=true
    overridePhysicsAsset=&/M_MED_Sunrise_Palace_Pack_Physics
    simulationSpace=BaseBoneSpace
    alphaBoolBlend {blendOption=Linear}
    baseBoneRef {boneName=pack}
    componentAppliedLinearAccClamp=<3000 3000 3000>
    componentLinearAccScale=<0.2 0.2 0.4>
    componentLinearVelScale=<0.1 0.1 0.2>
    overrideWorldGravity {z=-980}
   }
   n2 = "Transform (Modify) Bone - Bone: dyn_bag_02"(ComponentPose=@n1, Alpha=1, Translation="0,0,0", Rotation="0,0,45", Scale="1,1,1") {rotationMode=BMM_Additive, alphaBoolBlend {blendOption=Linear}, boneToModify {boneName=dyn_bag_02}, rotation {roll=45}}
   n3 = Component To Local(ComponentPose=@n2)
   n4 = RigidBody(ComponentPose=Local To Component(LocalPose=Use cached pose 'SavedPose21755'())) {
    bClampLinearTranslationLimitToRefPose=true
    bOverrideWorldGravity=true
    overridePhysicsAsset=&M_MED_Sunrise_Palace_Pack_Physics
    simulationSpace=BaseBoneSpace
    alphaBoolBlend {blendOption=Linear}
    alphaScaleBiasClamp {bInterpResult=true, interpSpeedDecreasing=4, interpSpeedIncreasing=6}
    baseBoneRef {boneName=root}
    componentAppliedLinearAccClamp=<3000 3000 3000>
    componentLinearAccScale=<0.15 0.15 0.25>
    componentLinearVelScale=<0.15 0.15 0.25>
    overrideWorldGravity {z=-980}
   }
   n5 = Component To Local(ComponentPose=@n4)
   n6 = Get bIsSkydiveFloating()
   n7 = "Transform (Modify) Bone - Bone: root"(
    ComponentPose=Local To Component(LocalPose=Blend Poses by bool(BlendPose_0=@n3, BlendPose_1=@n5, BlendTime_1=1, bActiveValue=@n6) {blendType=Linear})
    Translation=Get Backpack Translation() -> (Backpack Translation:Vector)
    Rotation=Get Backpack Rotate() -> (Backpack Rotate:Rotator)
   ) {
    rotationMode=BMM_Additive
    rotationSpace=BCS_BoneSpace
    translationMode=BMM_Additive
    translationSpace=BCS_BoneSpace
    alphaBoolBlend {blendOption=Linear}
    boneToModify {boneName=root}
   }
   Output Pose(Result=Component To Local(ComponentPose=@n7)) {layerGroup=None}
   Save cached pose 'SavedPose21755'(Pose=Input Pose()) {cacheName=SavedPose21755}
  ]}
 EventGraph {nodes [
   Event BlueprintUpdateAnimation()
   n1 = GetOwnerSkeletalMeshComponent()
   n2 = BreakRotator(InRot=BreakTransform(InTransform=GetDeltaTransformFromRefPose(self=@n1, BoneName=spine_05, BaseName=spine_01)).Rotation)
   > Set Backpack Rotate(Backpack Rotate:Rotator=MakeRotator(Roll=MapRangeClamped(Value=@n2.Yaw, InRangeA=0, InRangeB=-30, OutRangeB=20))) -> (Output_Get:Rotator)
   > Set Backpack Translation(Backpack Translation:Vector=MakeVector(Y=MapRangeClamped(Value=@n2.Yaw, InRangeA=0, InRangeB=-30, OutRangeB=-3))) -> (Output_Get:Vector)
   TryGetPawnOwner()
  ]}
]
variables [
 Backpack Rotate {type=Rotator}
 Backpack Translation {type=Vector}
 Jump Velocity {type=Vector, default=<0 0 5000>}
]
